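# Unlit, masked, two-sided Niagara mesh-particle material for the Sakura drift trail's tyre effect, previewed on a drift-trail wheel mesh.
# A tiled tyre texture pans diagonally at a dynamic-parameter speed, and its desaturated colour plus scaled green channel is tinted by particle colour for emissive.
# Red channel times particle alpha feeds both a depth-faded product and a hard step cutout multiplied by a shared wheel-wrap mask.
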Material /VehicleCosmetics/DriftTrail/DriftTrail_Sakura/Material/M_TyreSakura
graph {
 properties {
  bAutomaticallySetUsageInEditor=false
  bUsedWithNiagaraMeshParticles=true
  blendMode=BLEND_Masked
  pixelDepthOffsetMode=PDOM_Legacy
  previewMesh=/VehicleCosmetics/Mutable/FX_Meshes/SM_DriftTrail_Wheel
  shadingModel=MSM_Unlit
  twoSided=true
  usedShadingModels=Unlit
 }
 nodes [
  n1 = DynamicParameter() {defaultValue=<rgba 1 0 1 1>, paramNames [X,Y,Param3,Param4]}
  n2 = TextureSampleParameter2D(UVs=Panner(Coordinate=TextureCoordinate() {uTiling=5}, Speed=AppendVector(A=@n1.X, B=@n1.X))) {parameterName=Texture, texture=/VehicleCosmetics/DriftTrail/DriftTrail_Sakura/Texture/Tyre_Texture}
  n3 = Desaturation(None=@n2.RGB)
  n4 = Multiply(A=@n2.G, B=@n1.Param3)
  n5 = ParticleColor()
  MaterialFunctionCall(Emissive=Multiply(A=Add(A=@n3, B=@n4), B=@n5.RGB)) {materialFunction=&MF_FX_EmissiveAdjustments}
  n6 = Multiply(A=Add(A=@n2.R, B=@n2.R), B=@n5.A)
  Multiply(
   A=@n6
   B=MaterialFunctionCall(Opacity In=Constant() {r=20}) {materialFunction=&MF_VCVFX_DepthFade}.Result
  )
  n7 = Multiply(A=@n6) {constB=4}
  Multiply(
   A=MaterialFunctionCall!(Multiply, Power EXP=Constant() {r=2.1}) {materialFunction=/DealershipSharedLibrary/FX/Materials/MaterialFunctions/MF_VCVFX_WheelwrapMask}.Result
   B=Step(X=Saturate(None=@n7)) {constY=0.5}
  )
 ]
}
